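# Physics asset for the face accessory part of the StaticShades male medium player cosmetic, giving physics bodies to a dangling scarf bone chain.
# Sphere and capsule bodies cover the scarf chain, with larger bodies on the upper spine and left clavicle as torso shapes.
# Constraints link the scarf from its master bone through three segments with tight twist and first-swing limits but wider second-swing limits, while the clavicle-to-spine constraint is free on all angular axes.
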PhysicsAsset /BRCosmetics/Characters/Player/Male/Medium/Bodies/M_MED_StaticShades/Meshes/Parts/M_MED_StaticShades_FaceAcc_Physics
bodies {
 clavicle_l [{shapeType=Sphere, radius=10.2854, shapeName=clavicle_l_sphere, center=<12.1376 -1.3389 -5.5601>}]
 dyn_scarf_01 [{shapeType=Capsule, length=8, radius=5.2, shapeName=dyn_scarf_01_capsule, center=<5.8756 0 -6.7799>, rotation=<rot 0 90 93.0001>}]
 dyn_scarf_02 [{shapeType=Sphere, radius=5.4287, shapeName=dyn_scarf_02_sphere, center=<6.1608 -0.9942 -5.11>}]
 dyn_scarf_03 [{shapeType=Sphere, radius=4.7581, shapeName=dyn_scarf_03_sphere, center=<4.6808 2.6637 -6.0623>}]
 dyn_scarf_MSTR [{shapeType=Sphere, radius=2.7357, shapeName=dyn_scarf_MSTR_sphere}]
 spine_04 [{shapeType=Sphere, radius=15, shapeName=spine_04_sphere, center=<1.3228 1.2129 0>}]
 spine_05 [{shapeType=Capsule, length=15, radius=15.9757, shapeName=spine_05_capsule, center=<0.0913 3.8407 0>}]
}
constraints [
 {bone1Name=dyn_scarf_MSTR, bone2Name=dyn_scarf_01, swing1LimitDegrees=5, swing2LimitDegrees=40, twistLimitDegrees=5}
 {
  bone1Name=clavicle_l
  bone2Name=spine_05
  swing1LimitDegrees=45
  swing1Motion=Free
  swing2LimitDegrees=45
  swing2Motion=Free
  twistLimitDegrees=45
  twistMotion=Free
 }
 {bone1Name=dyn_scarf_02, bone2Name=dyn_scarf_01, swing1LimitDegrees=5, swing2LimitDegrees=60, twistLimitDegrees=5}
 {bone1Name=dyn_scarf_03, bone2Name=dyn_scarf_02, swing1LimitDegrees=5, swing2LimitDegrees=30, twistLimitDegrees=5}
]
