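# Anim Blueprint for the StaticShades male face accessory part, animating its scarf over a pose copied from the parent mesh, with clavicle branches from the body mesh.
# It selects RigidBody, AnimDynamics or a simple Trail fallback; the first two add weighted ring constraints, springs and skydive/aim corrections.
# Wearing a backpack zeroes the main scarf simulation alpha.
# The event graph updates gravity, pose alphas and cloth WPO/wind material parameters.
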AnimBlueprint /BRCosmetics/Characters/Player/Male/Medium/Bodies/M_MED_StaticShades/Meshes/Parts/M_MED_StaticShades_FaceAcc_AnimBP
target_skeleton=&/M_MED_StaticShades_FaceAcc_Skeleton
graphs [
 AnimGraph {nodes [
   n1 = Trail controller - Bone: dyn_scarf_02(ComponentPose=Local To Component(LocalPose=Use cached pose 'Start_Cache'()), Alpha=1) {
    bLimitRotation=true
    bShowBaseMotion=false
    bUsePlanarLimit=true
    planarLimits [{drivingBone {boneName=spine_05}, planeTransform {location=<0 -13 0>, rotation=<rot 0 20 -90>, scale=<1 1 1>}}]
    rotationLimits [{limitMax=<90 90 90>, limitMin=<-90 -90 -90>}, {limitMax=<180 180 180>}]
    rotationOffsets [<0 0 0>, <0 0 0>]
    trailBone {boneName=dyn_scarf_02}
    trailRelaxationSpeed {editorCurveData {keys [<at 1 3 RCIM_Linear>,<at 2 1 RCIM_Linear>]}}
   }
   n2 = Component To Local(ComponentPose=@n1)
   n3 = Anim Dynamics - Bone: dyn_scarf_01!(
    ComponentPose=Local To Component(LocalPose=Use cached pose 'Start_Cache'())
    Alpha=Get HasBackpackAlpha() -> (HasBackpackAlpha:double)
    GravityOverride:Vector=Get Gravity Override()
    ExternalForce="0,0,0"
   ) -> (Pose) {
    angularDampingOverride=0.998
    bGravityOverrideInSimSpace=true
    bOverrideAngularDamping=true
    bUseGravityOverride=true
    simulationSpace=RootRelative
    alphaBoolBlend {blendOption=Linear}
    boundBone {boneName=dyn_scarf_01}
    componentAppliedLinearAccClamp=<5000 5000 5000>
    componentLinearAccScale=<0.3 0.3 0.4>
    componentLinearVelScale=<0.3 0.3 0.3>
    gravityOverride {z=-980}
    physicsBodyDefinitions [{
      sphereCollisionRadius=0
      boundBone {boneName=dyn_scarf_01}
      boxExtents=<12 12 5>
      constraintSetup {angularLimitsMax=<2 100 2>, angularLimitsMin=<-2 10 -2>, angularTarget=<0 0 0>}
      localJointOffset=<4 0 0>
     }]
    relativeSpaceBone {boneName=dyn_scarf_MSTR}
   }
   n4 = "Constraint - dyn_scarf_ring_front_MSTR (PPPPPP)"!(
    ComponentPose=@n3
    Alpha=1
    ConstraintWeights_0=0.024
    ConstraintWeights_1:float=0
    ConstraintWeights_2:float=0
    ConstraintWeights_3:float=0.089
    ConstraintWeights_4:float=0.328
    ConstraintWeights_5:float=0.559
   ) -> (Pose) {
    alphaBoolBlend {blendOption=Linear}
    boneToModify {boneName=dyn_scarf_ring_front_MSTR}
    constraintSetup [
     {targetBone {boneName=clavicle_l}}
     {targetBone {boneName=clavicle_r}}
     {targetBone {boneName=neck_02}}
     {targetBone {boneName=spine_03}}
     {targetBone {boneName=spine_04}}
     {targetBone {boneName=spine_05}}
    ]
    constraintWeights [1,1,1,1,1,1]
   }
   n5 = "Constraint - dyn_scarf_ring_l_MSTR (PP)"!(ComponentPose=@n4, Alpha=1, ConstraintWeights_0=0.86, ConstraintWeights_1:float=0.14) -> (Pose) {
    alphaBoolBlend {blendOption=Linear}
    boneToModify {boneName=dyn_scarf_ring_l_MSTR}
    constraintSetup [{targetBone {boneName=clavicle_l}}, {targetBone {boneName=spine_05}}]
    constraintWeights [1,1]
   }
   n6 = "Constraint - dyn_scarf_ring_l_upper_MSTR (PPPPPP)"!(
    ComponentPose=@n5
    Alpha=1
    ConstraintWeights_0=0.111
    ConstraintWeights_1:float=0.04
    ConstraintWeights_2:float=0.356
    ConstraintWeights_3:float=0.321
    ConstraintWeights_4:float=0
    ConstraintWeights_5:float=0.172
   ) -> (Pose) {
    alphaBoolBlend {blendOption=Linear}
    boneToModify {boneName=dyn_scarf_ring_l_upper_MSTR}
    constraintSetup [
     {targetBone {boneName=clavicle_l}}
     {targetBone {boneName=head}}
     {targetBone {boneName=neck_01}}
     {targetBone {boneName=neck_02}}
     {targetBone {boneName=spine_04}}
     {targetBone {boneName=spine_05}}
    ]
    constraintWeights [1,1,1,1,1,1]
   }
   n7 = "Constraint - dyn_scarf_ring_back_MSTR (PPPPPP)"!(
    ComponentPose=@n6
    Alpha=1
    ConstraintWeights_0=0.267
    ConstraintWeights_1:float=0.04
    ConstraintWeights_2:float=0.03
    ConstraintWeights_3:float=0.597
    ConstraintWeights_4:float=0.02
    ConstraintWeights_5:float=0.046
   ) -> (Pose) {
    alphaBoolBlend {blendOption=Linear}
    boneToModify {boneName=dyn_scarf_ring_back_MSTR}
    constraintSetup [
     {targetBone {boneName=clavicle_l}}
     {targetBone {boneName=spine_03}}
     {targetBone {boneName=spine_04}}
     {targetBone {boneName=spine_05}}
     {targetBone {boneName=neck_01}}
     {targetBone {boneName=neck_02}}
    ]
    constraintWeights [1,1,1,1,1,1]
   }
   n8 = Anim Dynamics - Bone: dyn_scarf_ring_l_upper!(ComponentPose=@n7, Alpha=1, GravityOverride:Vector="0,0,-980", ExternalForce="0,0,0") -> (Pose) {
    angularDampingOverride=0.9
    angularSpringConstant=1000
    bAngularSpring=true
    bGravityOverrideInSimSpace=true
    bOverrideAngularDamping=true
    bUseGravityOverride=true
    simulationSpace=RootRelative
    alphaBoolBlend {blendOption=Linear}
    boundBone {boneName=dyn_scarf_ring_l_upper}
    componentAppliedLinearAccClamp=<2000 2000 2000>
    componentLinearAccScale=<0.01 0.01 0.01>
    componentLinearVelScale=<0.01 0.01 0.01>
    gravityOverride {z=-980}
    physicsBodyDefinitions [{
      sphereCollisionRadius=0
      boundBone {boneName=dyn_scarf_ring_l_upper}
      boxExtents=<2 4 4>
      constraintSetup {angularLimitsMax=<5 0 5>, angularLimitsMin=<-5 0 -5>, angularTarget=<0 0 0>}
      localJointOffset=<-1 0 0>
     }]
    relativeSpaceBone {boneName=dyn_scarf_MSTR}
   }
   n9 = Anim Dynamics - Bone: dyn_scarf_ring_l!(ComponentPose=@n8, Alpha=1, GravityOverride:Vector="0,0,-980", ExternalForce="0,0,0") -> (Pose) {
    angularDampingOverride=1
    angularSpringConstant=3000
    bAngularSpring=true
    bGravityOverrideInSimSpace=true
    bOverrideAngularDamping=true
    bUseGravityOverride=true
    simulationSpace=RootRelative
    alphaBoolBlend {blendOption=Linear}
    boundBone {boneName=dyn_scarf_ring_l}
    componentAppliedLinearAccClamp=<2000 2000 2000>
    componentLinearAccScale=<0.01 0.01 0.01>
    componentLinearVelScale=<0.01 0.01 0.01>
    gravityOverride {z=-980}
    physicsBodyDefinitions [{
      sphereCollisionRadius=0
      boundBone {boneName=dyn_scarf_ring_l}
      boxExtents=<4 4 2>
      constraintSetup {angularLimitsMax=<5 5 0>, angularLimitsMin=<-5 -10 0>, angularTarget=<0 0 0>}
      localJointOffset=<-0.5 0 0>
     }]
    relativeSpaceBone {boneName=dyn_scarf_MSTR}
   }
   n10 = Anim Dynamics - Bone: dyn_scarf_ring_front!(ComponentPose=@n9, Alpha=1, GravityOverride:Vector="0,0,-980", ExternalForce="0,0,0") -> (Pose) {
    angularDampingOverride=1
    angularSpringConstant=3000
    bAngularSpring=true
    bGravityOverrideInSimSpace=true
    bOverrideAngularDamping=true
    bUseGravityOverride=true
    simulationSpace=RootRelative
    alphaBoolBlend {blendOption=Linear}
    boundBone {boneName=dyn_scarf_ring_front}
    componentAppliedLinearAccClamp=<2000 2000 2000>
    componentLinearAccScale=<0.01 0.01 0.01>
    componentLinearVelScale=<0.01 0.01 0.01>
    gravityOverride {z=-980}
    physicsBodyDefinitions [{
      sphereCollisionRadius=0
      boundBone {boneName=dyn_scarf_ring_front}
      boxExtents=<2 4 4>
      constraintSetup {angularLimitsMax=<5 0 5>, angularLimitsMin=<-5 0 -5>, angularTarget=<0 0 0>}
      localJointOffset=<-1 0 0>
     }]
    relativeSpaceBone {boneName=dyn_scarf_MSTR}
   }
   n11 = Anim Dynamics - Bone: dyn_scarf_ring_back!(ComponentPose=@n10, Alpha=1, GravityOverride:Vector="0,0,-980", ExternalForce="0,0,0") -> (Pose) {
    angularDampingOverride=1
    angularSpringConstant=3000
    bAngularSpring=true
    bGravityOverrideInSimSpace=true
    bOverrideAngularDamping=true
    bUseGravityOverride=true
    simulationSpace=RootRelative
    alphaBoolBlend {blendOption=Linear}
    boundBone {boneName=dyn_scarf_ring_back}
    componentAppliedLinearAccClamp=<2000 2000 2000>
    componentLinearAccScale=<0.01 0.01 0.01>
    componentLinearVelScale=<0.01 0.01 0.01>
    gravityOverride {z=-980}
    physicsBodyDefinitions [{
      sphereCollisionRadius=0
      boundBone {boneName=dyn_scarf_ring_back}
      boxExtents=<2 4 4>
      constraintSetup {angularLimitsMax=<5 5 0>, angularLimitsMin=<-5 -10 0>, angularTarget=<0 0 0>}
      localJointOffset=<-1 0 0>
     }]
    relativeSpaceBone {boneName=dyn_scarf_MSTR}
   }
   n12 = Get Skydiving Alpha() -> (Skydiving Alpha:double)
   n13 = "Transform (Modify) Bone - Bone: dyn_scarf_01"(ComponentPose=@n11, Alpha=@n12, Translation="0,0,0", Rotation="-20,0,0", Scale="1,1,1") {
    rotationMode=BMM_Additive
    rotationSpace=BCS_BoneSpace
    translationMode=BMM_Additive
    translationSpace=BCS_BoneSpace
    alphaBoolBlend {blendOption=Linear}
    alphaScaleBiasClamp {bInterpResult=true}
    boneToModify {boneName=dyn_scarf_01}
    rotation {pitch=-20}
   }
   n14 = "Transform (Modify) Bone - Bone: dyn_scarf_02"(ComponentPose=@n13, Alpha=@n12, Translation="0,0,0", Rotation="-30,0,0", Scale="1,1,1") {
    rotationMode=BMM_Additive
    rotationSpace=BCS_BoneSpace
    translationMode=BMM_Additive
    translationSpace=BCS_BoneSpace
    alphaBoolBlend {blendOption=Linear}
    alphaScaleBiasClamp {bInterpResult=true}
    boneToModify {boneName=dyn_scarf_02}
    rotation {pitch=-30}
   }
   n15 = "Transform (Modify) Bone - Bone: deform_neck_ADS_compensator"(
    ComponentPose=@n14
    Alpha=Get TargetingHeadUpAlpha() -> (TargetingHeadUpAlpha:double)
    Translation="0,0,-2.4"
    Rotation="0,0,0"
    Scale="1,1,1"
   ) {
    rotationMode=BMM_Additive
    rotationSpace=BCS_BoneSpace
    translationMode=BMM_Additive
    translationSpace=BCS_BoneSpace
    alphaBoolBlend {blendInTime=0.3, blendOption=ExpInOut, blendOutTime=0.3}
    boneToModify {boneName=deform_neck_ADS_compensator}
    translation {z=-2.4}
   }
   n16 = Component To Local(ComponentPose=@n15)
   n17 = Get bAnimDynamicsEnabled() -> (bAnimDynamicsEnabled:bool)
   n18 = RigidBody!(
    ComponentPose=Local To Component(LocalPose=Use cached pose 'Start_Cache'())
    Alpha=Get HasBackpackAlpha()
    OverrideWorldGravity:Vector=Get Gravity Override()
    ExternalForce
   ) -> (Pose) {
    bClampLinearTranslationLimitToRefPose=true
    bOverrideWorldGravity=true
    overridePhysicsAsset=&/M_MED_StaticShades_FaceAcc_Physics
    simulationSpace=BaseBoneSpace
    alphaBoolBlend {blendOption=Linear}
    baseBoneRef {boneName=spine_05}
    componentAppliedLinearAccClamp=<2000 2000 2000>
    componentLinearAccScale=<0.4 0.4 0.4>
    componentLinearVelScale=<0.6 0.6 0.6>
    overrideWorldGravity {z=-980}
   }
   n19 = "Constraint - dyn_scarf_ring_front_MSTR (PPPPPP)"(
    ComponentPose=@n18
    Alpha=1
    ConstraintWeights_0=0.024
    ConstraintWeights_1=0
    ConstraintWeights_2=0
    ConstraintWeights_3=0.089
    ConstraintWeights_4=0.328
    ConstraintWeights_5=0.559
   ) {
    alphaBoolBlend {blendOption=Linear}
    boneToModify {boneName=dyn_scarf_ring_front_MSTR}
    constraintSetup [
     {targetBone {boneName=clavicle_l}}
     {targetBone {boneName=clavicle_r}}
     {targetBone {boneName=neck_02}}
     {targetBone {boneName=spine_03}}
     {targetBone {boneName=spine_04}}
     {targetBone {boneName=spine_05}}
    ]
    constraintWeights [1,1,1,1,1,1]
   }
   n20 = "Constraint - dyn_scarf_ring_l_MSTR (PP)"(ComponentPose=@n19, Alpha=1, ConstraintWeights_0=0.86, ConstraintWeights_1=0.14) {
    alphaBoolBlend {blendOption=Linear}
    boneToModify {boneName=dyn_scarf_ring_l_MSTR}
    constraintSetup [{targetBone {boneName=clavicle_l}}, {targetBone {boneName=spine_05}}]
    constraintWeights [1,1]
   }
   n21 = "Constraint - dyn_scarf_ring_l_upper_MSTR (PPPPPP)"(
    ComponentPose=@n20
    Alpha=1
    ConstraintWeights_0=0.111
    ConstraintWeights_1=0.04
    ConstraintWeights_2=0.356
    ConstraintWeights_3=0.321
    ConstraintWeights_4=0
    ConstraintWeights_5=0.172
   ) {
    alphaBoolBlend {blendOption=Linear}
    boneToModify {boneName=dyn_scarf_ring_l_upper_MSTR}
    constraintSetup [
     {targetBone {boneName=clavicle_l}}
     {targetBone {boneName=head}}
     {targetBone {boneName=neck_01}}
     {targetBone {boneName=neck_02}}
     {targetBone {boneName=spine_04}}
     {targetBone {boneName=spine_05}}
    ]
    constraintWeights [1,1,1,1,1,1]
   }
   n22 = "Constraint - dyn_scarf_ring_back_MSTR (PPPPPP)"(
    ComponentPose=@n21
    Alpha=1
    ConstraintWeights_0=0.267
    ConstraintWeights_1=0.04
    ConstraintWeights_2=0.03
    ConstraintWeights_3=0.597
    ConstraintWeights_4=0.02
    ConstraintWeights_5=0.046
   ) {
    alphaBoolBlend {blendOption=Linear}
    boneToModify {boneName=dyn_scarf_ring_back_MSTR}
    constraintSetup [
     {targetBone {boneName=clavicle_l}}
     {targetBone {boneName=spine_03}}
     {targetBone {boneName=spine_04}}
     {targetBone {boneName=spine_05}}
     {targetBone {boneName=neck_01}}
     {targetBone {boneName=neck_02}}
    ]
    constraintWeights [1,1,1,1,1,1]
   }
   n23 = Anim Dynamics - Bone: dyn_scarf_ring_l_upper(ComponentPose=@n22, Alpha=1, GravityOverride="0,0,-980", ExternalForce="0,0,0") {
    angularDampingOverride=1
    angularSpringConstant=2500
    bAngularSpring=true
    bGravityOverrideInSimSpace=true
    bOverrideAngularDamping=true
    bUseGravityOverride=true
    bUsePlanarLimit=false
    simulationSpace=BoneRelative
    alphaBoolBlend {blendOption=Linear}
    boundBone {boneName=dyn_scarf_ring_l_upper}
    componentAppliedLinearAccClamp=<2000 2000 2000>
    componentLinearAccScale=<0.1 0.1 0.1>
    componentLinearVelScale=<0.1 0.1 0.1>
    gravityOverride {z=-980}
    physicsBodyDefinitions [{
      sphereCollisionRadius=0
      boundBone {boneName=dyn_scarf_ring_l_upper}
      boxExtents=<2 4 4>
      constraintSetup {angularLimitsMax=<5 0 5>, angularLimitsMin=<-5 0 -5>, angularTarget=<0 0 0>}
      localJointOffset=<-1 0 0>
     }]
    relativeSpaceBone {boneName=dyn_scarf_MSTR}
   }
   n24 = Anim Dynamics - Bone: dyn_scarf_ring_l(ComponentPose=@n23, Alpha=1, GravityOverride="0,0,-980", ExternalForce="0,0,0") {
    angularDampingOverride=1
    angularSpringConstant=2500
    bAngularSpring=true
    bGravityOverrideInSimSpace=true
    bOverrideAngularDamping=true
    bUseGravityOverride=true
    bUsePlanarLimit=false
    simulationSpace=BoneRelative
    alphaBoolBlend {blendOption=Linear}
    boundBone {boneName=dyn_scarf_ring_l}
    componentAppliedLinearAccClamp=<2000 2000 2000>
    componentLinearAccScale=<0.1 0.1 0.1>
    componentLinearVelScale=<0.1 0.1 0.1>
    gravityOverride {z=-980}
    physicsBodyDefinitions [{
      sphereCollisionRadius=0
      boundBone {boneName=dyn_scarf_ring_l}
      boxExtents=<4 4 2>
      constraintSetup {angularLimitsMax=<5 5 0>, angularLimitsMin=<-5 -10 0>, angularTarget=<0 0 0>}
      localJointOffset=<-0.5 0 0>
     }]
    relativeSpaceBone {boneName=dyn_scarf_MSTR}
   }
   n25 = Anim Dynamics - Bone: dyn_scarf_ring_front(ComponentPose=@n24, Alpha=1, GravityOverride="0,0,-980", ExternalForce="0,0,0") {
    angularDampingOverride=1
    angularSpringConstant=2500
    bAngularSpring=true
    bGravityOverrideInSimSpace=true
    bOverrideAngularDamping=true
    bUseGravityOverride=true
    bUsePlanarLimit=false
    simulationSpace=BoneRelative
    alphaBoolBlend {blendOption=Linear}
    boundBone {boneName=dyn_scarf_ring_front}
    componentAppliedLinearAccClamp=<2000 2000 2000>
    componentLinearAccScale=<0.1 0.1 0.1>
    componentLinearVelScale=<0.1 0.1 0.1>
    gravityOverride {z=-980}
    physicsBodyDefinitions [{
      sphereCollisionRadius=0
      boundBone {boneName=dyn_scarf_ring_front}
      boxExtents=<2 4 4>
      constraintSetup {angularLimitsMax=<5 0 5>, angularLimitsMin=<-5 0 -5>, angularTarget=<0 0 0>}
      localJointOffset=<-1 0 0>
     }]
    relativeSpaceBone {boneName=dyn_scarf_MSTR}
   }
   n26 = Anim Dynamics - Bone: dyn_scarf_ring_back(ComponentPose=@n25, Alpha=1, GravityOverride="0,0,-980", ExternalForce="0,0,0") {
    angularDampingOverride=1
    angularSpringConstant=2500
    bAngularSpring=true
    bGravityOverrideInSimSpace=true
    bOverrideAngularDamping=true
    bUseGravityOverride=true
    bUsePlanarLimit=false
    simulationSpace=BoneRelative
    alphaBoolBlend {blendOption=Linear}
    boundBone {boneName=dyn_scarf_ring_back}
    componentAppliedLinearAccClamp=<2000 2000 2000>
    componentLinearAccScale=<0.1 0.1 0.1>
    componentLinearVelScale=<0.1 0.1 0.1>
    gravityOverride {z=-980}
    physicsBodyDefinitions [{
      sphereCollisionRadius=0
      boundBone {boneName=dyn_scarf_ring_back}
      boxExtents=<2 4 4>
      constraintSetup {angularLimitsMax=<5 5 0>, angularLimitsMin=<-5 -10 0>, angularTarget=<0 0 0>}
      localJointOffset=<-1 0 0>
     }]
    relativeSpaceBone {boneName=dyn_scarf_MSTR}
   }
   n27 = "Transform (Modify) Bone - Bone: dyn_scarf_MSTR"(
    ComponentPose=@n26
    Alpha=Get LeftShoulderBackwardMovementAlpha() -> (LeftShoulderBackwardMovementAlpha:double)
    Translation="0,0,0"
    Rotation="-20,0,0"
    Scale="1,1,1"
   ) {
    rotationMode=BMM_Additive
    rotationSpace=BCS_BoneSpace
    translationMode=BMM_Additive
    translationSpace=BCS_BoneSpace
    alphaBoolBlend {blendOption=Linear}
    boneToModify {boneName=dyn_scarf_MSTR}
    rotation {pitch=-20}
   }
   n28 = "Transform (Modify) Bone - Bone: dyn_scarf_MSTR"(
    ComponentPose=@n27
    Alpha=Get Forward/Backward Compensation alpha() -> (Forward/Backward Compensation alpha:double)
    Translation="0,0,-5"
    Rotation="0,0,0"
    Scale="1,1,1"
   ) {
    rotationMode=BMM_Additive
    rotationSpace=BCS_BoneSpace
    translationMode=BMM_Additive
    translationSpace=BCS_BoneSpace
    alphaBoolBlend {blendOption=Linear}
    boneToModify {boneName=dyn_scarf_MSTR}
    translation {z=-5}
   }
   n29 = Get Skydiving Alpha()
   n30 = "Transform (Modify) Bone - Bone: dyn_scarf_01"(ComponentPose=@n28, Alpha=@n29, Translation="0,0,0", Rotation="-20,0,0", Scale="1,1,1") {
    rotationMode=BMM_Additive
    rotationSpace=BCS_BoneSpace
    translationMode=BMM_Additive
    translationSpace=BCS_BoneSpace
    alphaBoolBlend {blendOption=Linear}
    alphaScaleBiasClamp {bInterpResult=true}
    boneToModify {boneName=dyn_scarf_01}
    rotation {pitch=-20}
   }
   n31 = "Transform (Modify) Bone - Bone: dyn_scarf_02"(ComponentPose=@n30, Alpha=@n29, Translation="0,0,0", Rotation="-30,0,0", Scale="1,1,1") {
    rotationMode=BMM_Additive
    rotationSpace=BCS_BoneSpace
    translationMode=BMM_Additive
    translationSpace=BCS_BoneSpace
    alphaBoolBlend {blendOption=Linear}
    alphaScaleBiasClamp {bInterpResult=true}
    boneToModify {boneName=dyn_scarf_02}
    rotation {pitch=-30}
   }
   n32 = "Transform (Modify) Bone - Bone: deform_neck_ADS_compensator"(ComponentPose=@n31, Alpha=Get TargetingHeadUpAlpha(), Translation="0,0,-2.5", Rotation="-5,0,11", Scale="1,1,1") {
    rotationMode=BMM_Additive
    rotationSpace=BCS_BoneSpace
    translationMode=BMM_Additive
    translationSpace=BCS_BoneSpace
    alphaBoolBlend {blendInTime=0.3, blendOption=ExpInOut, blendOutTime=0.3}
    boneToModify {boneName=deform_neck_ADS_compensator}
    rotation {pitch=-5, roll=11}
    translation {z=-2.5}
   }
   n33 = "Transform (Modify) Bone - Bone: dyn_scarf_03"!(ComponentPose=@n32, bAlphaBoolEnabled:bool=Get Is in air() -> (Is in air:bool), Translation="0,0,0", Rotation="-50,0,0", Scale="1,1,1") -> (Pose) {
    alphaInputType=Bool
    rotationMode=BMM_Additive
    rotationSpace=BCS_BoneSpace
    translationMode=BMM_Additive
    translationSpace=BCS_BoneSpace
    alphaBoolBlend {blendInTime=0.3, blendOption=ExpInOut, blendOutTime=0.3}
    boneToModify {boneName=dyn_scarf_03}
    rotation {pitch=-50}
   }
   Output Pose(Result=Blend Poses by bool(
     BlendPose_0=Component To Local(ComponentPose=@n33)
     BlendPose_1=Blend Poses by bool(BlendPose_0=@n16, BlendPose_1=@n2, bActiveValue=@n17) {blendType=Linear}
     bActiveValue=Get bRigidBodyEnabled()
    ) {blendType=Linear}) {layerGroup=None}
   n34 = Get Body Mesh() -> (Body Mesh:Skeletal Mesh Component Object Reference)
   Save cached pose 'Start_Cache'(Pose=Layered blend per bone(BasePose=Copy Pose From Mesh() {bUseAttachedParent=true}, BlendPoses_0=Copy Pose From Mesh(SourceMeshComponent=@n34)) {layerSetup [{branchFilters [{boneName=clavicle_r}, {boneName=clavicle_l}]}]}) {cacheName=Start_Cache}
  ]}
 EventGraph {nodes [
   n1 = Calculate Velocity From Position History!(
    execute
    DeltaSeconds:float=Get Delta Time X()
    Position:Vector=Get Position() -> (Position:Vector)
    History:"Position History Structure (by ref)"=Get History() -> (History:Position History Structure)
    NumberOfSamples:int=16
    VelocityMin:float=0.f
    VelocityMax:float=128.f
   ) -> (then, ReturnValue:float)
   > Set Forward/Backward Compensation alpha(Forward/Backward Compensation alpha:double=@n1) -> (Output_Get:double)
   n2 = Event BlueprintUpdateAnimation()
   > Set Delta Time X(Delta Time X=@n2.DeltaTimeX)
   > n3 = Sequence!(execute) -> (then_0, then_1, then_2:Exec, then_3:Exec, then_4:Exec)
   > n4 = Gravity Override State Function DEPRECATED(CharacterPartAnimInstance=Self-Reference(), GameTimeInSeconds=GetGameTimeInSeconds(), GravityOverrideParams=Get Gravity Override Params())
   n5 = BreakVector(InVec=@n4)
   n6 = NOT Boolean(A=Get bIsSkyDiving())
   n7 = Get bIsJumping()
   n8 = OR Boolean(A=@n7, B=Get bIsFalling())
   > Set Gravity Override(Gravity Override=SelectVector(A=MakeVector(X=@n5, Y=@n5.Y, Z=-200), B=@n4, bPickA=AND Boolean(A=@n8, B=@n6)))
   > Set Skydiving Alpha(Skydiving Alpha:double=SelectFloat(A=1, B=0, bPickA=Get bIsSkyDiving())) -> (Output_Get:double)
   n9 = BreakTransform(InTransform=GetDeltaTransformFromRefPose(self=GetOwnerSkeletalMeshComponent(), BoneName=neck_01, BaseName=root))
   Set TargetingHeadUpAlpha(execute=@n3.then_3, TargetingHeadUpAlpha:double=MapRangeClamped(Value=BreakRotator(InRot=@n9.Rotation).Yaw, InRangeA=5, InRangeB=-30)) -> (Output_Get:double)
   Set Is in air(execute=@n3.then_4, Is in air:bool=OR Boolean(A=Get bIsParachuteOpened(), B=@n7)) -> (Output_Get:bool)
   Set LeftShoulderBackwardMovementAlpha(
    execute=@n3.then_1
    LeftShoulderBackwardMovementAlpha:double=GetAlphaFromReferenceTransformDelta(SkeletalMeshComponent=Get Body Mesh(), BoneName=clavicle_l, BaseName=spine_05, InputRange=(X=0,Y=-20), TwistAxis="1,0,0")
   ) -> (Output_Get:double)
   n10 = TryGetPawnOwner()
   n11 = Dot Product!(
    A:Vector=GetVelocity!(self:Actor Object Reference=@n10) -> (ReturnValue:Vector)
    B:Vector=GetActorForwardVector!(self:Actor Object Reference=@n10) -> (ReturnValue:Vector)
   ) -> (ReturnValue:double)
   > Set Forward or BackwardMovementAlpha(Forward or BackwardMovementAlpha:double=MapRangeClamped(Value="Absolute (Float)"(A=@n11), InRangeA=0, InRangeB=200)) -> (Output_Get:double)
   Is Valid(exec=@n3.then_2, InputObject=Get PlayerPawn() -> (PlayerPawn:Fort Player Pawn Athena Object Reference))
   > n12 = Branch(Condition=Get bIsInFrontEnd())
   Set WPO Globals Function(
    execute=@n12.else
    Material=Get Mat Instance Dynamic()
    Overlay Material=Get Overlay Mat Instance Dynamic()
    Skeletal Mesh=Get Skel Mesh()
    "Local Noise Transform  Location"="Get Local Noise Transform  Location"() -> ("Local Noise Transform  Location":Vector)
    Local Noise Offset=Get Local Noise Offset()
   )
   > Set WPO Regional Function!(
    execute
    Material:Material Instance Dynamic Object Reference=Get Mat Instance Dynamic()
    Overlay Material:Material Instance Dynamic Object Reference=Get Overlay Mat Instance Dynamic()
    Skeletal Mesh:Skeletal Mesh Component Object Reference=Get Skel Mesh()
    Velocity Transform:Name=dyn_scarf_02
    Velocity Interp Speed:double=Get Velocity Interp Speed()
    Collision Transform:Name=clavicle_l
    Collision Transform Axis:EAxis Enum
    Invert Collision Transform Axis:bool=false
    Delta Time X:double=Get Delta Time X()
    Previous Velocity Transform Location:"Vector (by ref)"=Get R1 Velocity Transform Location()
    Previous Velocity Transform Velocity:"Vector (by ref)"=Get R1 Velocity Transform Velocity()
    Region:int=1
   ) -> (then)
   > n13 = Branch(Condition=Get bIsSkydiveFloating())
   > n14 = Set WPO Pho Wind Intensity(WPO Pho Wind Intensity=1)
   SetScalarParameterValue(execute=[@n15.then, @n14.then], self=Get Mat Instance Dynamic(), ParameterName=ClothSim_WindStrength, Value=Get WPO Pho Wind Intensity())
   n15 = Set WPO Pho Wind Intensity(execute=@n13.else, WPO Pho Wind Intensity=0)
   Event OnCharacterCustomizationLoaded()
   > n16 = Sequence!(execute) -> (then_0, then_1, then_2, then_3)
   > n17 = Set Skel Mesh(Skel Mesh=GetPartSkeletalMeshComponent())
   > n18 = Get Dynamic Material Instance(SkeletalMeshComponent=@n17, ElementIndex=2)
   > Set Mat Instance Dynamic(Mat Instance Dynamic=@n18)
   > n19 = GetOrCreateDynamicOverlayMaterial(MeshComponent=@n17, MaterialName=None)
   > Set Overlay Mat Instance Dynamic(Overlay Mat Instance Dynamic=@n19)
   n20 = CheckForBackbling()
   Set is wearing backpack(execute=@n16.then_1, is wearing backpack:bool=OR Boolean(A=@n20, B=@n20.bIsBackblingACape)) -> (Output_Get:bool)
   n21 = Cast To PlayerPawn_Athena(execute=@n16.then_2, Object=TryGetPawnOwner())
   > Set Body Mesh(Body Mesh:Skeletal Mesh Component Object Reference=GetSkeletalMeshForPartType(self=@n21)) -> (Output_Get:Skeletal Mesh Component Object Reference)
   n22 = Branch(execute=@n16.then_3, Condition=Get is wearing backpack() -> (is wearing backpack:bool))
   > Set HasBackpackAlpha(HasBackpackAlpha:double=0) -> (Output_Get:double)
   Set HasBackpackAlpha(execute=@n22.else, HasBackpackAlpha=1)
   Event BlueprintInitializeAnimation()
   n23 = TryGetPawnOwner()
   > Is Valid(InputObject=@n23)
   > n24 = Cast To FortPlayerPawnAthena(Object=@n23)
   > Set PlayerPawn(PlayerPawn:Fort Player Pawn Athena Object Reference=@n24) -> (Output_Get:Fort Player Pawn Athena Object Reference)
  ]}
]
variables [
 Body Mesh {type=SkeletalMeshComponent}
 Delta Time X {type=number}
 Forward or BackwardMovementAlpha {type=number}
 Forward/Backward Compensation alpha {type=number}
 Gravity Override {type=Vector}
 Gravity Override Params {
  type=GravityOverrideParamsStruct
  default {
   skydive_DivingDown {windFrequency=<6 12 10>}
   skydive_FloatingDown {gravityOverride=<0 -700 1500>, windAmplitude=<600 -1000 400>, windFrequency=<2 8 9>}
   skydive_GoingUp {gravityOverride=<0 0 0>}
   skydive_Parachuting {gravityOverride=<0 -800 -150>, windAmplitude=<200 -200 100>, windFrequency=<1.5 3 3>}
  }
 }
 HasBackpackAlpha {type=number}
 History {type=PositionHistory, default {range=0, positions []}}
 Is in air {type=boolean}
 is wearing backpack {type=boolean}
 LeftShoulderBackwardMovementAlpha {type=number}
 Local Noise Offset {type=Vector}
 "Local Noise Transform  Location" {type=Vector}
 Mat Instance Dynamic {type=MaterialInstanceDynamic}
 Overlay Mat Instance Dynamic {type=MaterialInstanceDynamic}
 PlayerPawn {type=FortPlayerPawnAthena}
 Position {type=Vector}
 R1 Velocity Transform Location {type=Vector}
 R1 Velocity Transform Velocity {type=Vector}
 R2 Velocity Transform Location {type=Vector}
 R2 Velocity Transform Velocity {type=Vector}
 Skel Mesh {type=SkeletalMeshComponent}
 Skydiving Alpha {type=number}
 TargetingHeadUpAlpha {type=number}
 Velocity Interp Speed {type=number, default=40}
 WPO Pho Wind Intensity {type=number}
]
